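# Animation Blueprint for the MusketSlinger male medium body cosmetic, adding secondary motion to its shoulder pads, pouch zipper and bootstraps.
# It copies the parent mesh pose, then runs the outfit's Control Rig.
# Additive bone offsets move each shoulder pad, weighted by upperarm-to-clavicle bend alphas computed every update.
# A weighted constraint ties the left pouch zipper to pelvis and thigh-area bones, then spring-driven AnimDynamics swings the zipper and both bootstraps.
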AnimBlueprint /BRCosmetics/Characters/Player/Male/Medium/Bodies/M_MED_MusketSlinger_Style/Meshes/M_MED_MusketSlinger_Style_AnimBP
target_skeleton=&/M_MED_MusketSlinger_Style_Skeleton
graphs [
 AnimGraph {nodes [
   n1 = Control Rig(Source=Copy Pose From Mesh() {bUseAttachedParent=true}) {
    alphaBoolBlend {blendOption=Linear}
    controlRigAssetReference {blueprintRigClass=&/M_MED_MusketSlinger_Style_CtrlRig.M_MED_MusketSlinger_Style_CtrlRig_C}
   }
   n2 = "Transform (Modify) Bone - Bone: deform_shoulderpad_l"(
    ComponentPose=Local To Component(LocalPose=@n1)
    Alpha=Get LeftArmBendEulerYAlpha() -> (LeftArmBendEulerYAlpha:double)
    Translation="2,-5,5"
    Rotation="30,30,50"
    Scale="1,1,1"
   ) {
    rotationMode=BMM_Additive
    rotationSpace=BCS_BoneSpace
    translationMode=BMM_Additive
    translationSpace=BCS_BoneSpace
    alphaBoolBlend {blendOption=Linear}
    boneToModify {boneName=deform_shoulderpad_l}
    rotation=<rot 30 30 50>
    translation=<2 -5 5>
   }
   n3 = "Transform (Modify) Bone - Bone: deform_shoulderpad_r"(
    ComponentPose=@n2
    Alpha=Get RightArmBendEulerYAlpha() -> (RightArmBendEulerYAlpha:double)
    Translation="2,5,5"
    Rotation="-30,30,50"
    Scale="1,1,1"
   ) {
    rotationMode=BMM_Additive
    rotationSpace=BCS_BoneSpace
    translationMode=BMM_Additive
    translationSpace=BCS_BoneSpace
    alphaBoolBlend {blendOption=Linear}
    boneToModify {boneName=deform_shoulderpad_r}
    rotation=<rot -30 30 50>
    translation=<2 5 5>
   }
   n4 = "Constraint - dyn_pouch_zipper_l_MSTR (PPPPP)"!(
    ComponentPose=@n3
    Alpha=1
    ConstraintWeights_0=0.119
    ConstraintWeights_1:float=0.682
    ConstraintWeights_2:float=0.041
    ConstraintWeights_3:float=0.072
    ConstraintWeights_4:float=0.086
   ) -> (Pose) {
    alphaBoolBlend {blendOption=Linear}
    boneToModify {boneName=dyn_pouch_zipper_l_MSTR}
    constraintSetup [
     {targetBone {boneName=pelvis}}
     {targetBone {boneName=thigh_twist_01_l}}
     {targetBone {boneName=deform_glute_l}}
     {targetBone {boneName=deform_groin_l}}
     {targetBone {boneName=thigh_l}}
    ]
    constraintWeights [1,1,1,1,1]
   }
   n5 = Anim Dynamics - Bone: dyn_pouch_zipper_l_01!(ComponentPose=@n4, Alpha=1, GravityOverride:Vector="0,0,-300", ExternalForce="0,0,0") -> (Pose) {
    angularDampingOverride=1
    angularSpringConstant=1000
    bAngularSpring=true
    bOverrideAngularDamping=true
    bUseGravityOverride=true
    simulationSpace=RootRelative
    alphaBoolBlend {blendOption=Linear}
    boundBone {boneName=dyn_pouch_zipper_l_01}
    componentAppliedLinearAccClamp=<2000 2000 2000>
    componentLinearAccScale=<0.1 0.1 0.1>
    componentLinearVelScale=<0.1 0.1 0.1>
    physicsBodyDefinitions [{
      boundBone {boneName=dyn_pouch_zipper_l_01}
      boxExtents=<5 5 5>
      constraintSetup {angularLimitsMax=<5 10 30>, angularLimitsMin=<-5 -30 -30>, angularTarget=<0 0 0>}
      localJointOffset=<2 0 0>
     }]
    relativeSpaceBone {boneName=spine_05}
   }
   n6 = Anim Dynamics - Bone: dyn_bootstrap_01_l!(ComponentPose=@n5, Alpha=1, GravityOverride:Vector="0,0,-300", ExternalForce="0,0,0") -> (Pose) {
    angularDampingOverride=1
    angularSpringConstant=1000
    bAngularSpring=true
    bOverrideAngularDamping=true
    bUseGravityOverride=true
    alphaBoolBlend {blendOption=Linear}
    boundBone {boneName=dyn_bootstrap_01_l}
    componentAppliedLinearAccClamp=<2000 2000 2000>
    componentLinearAccScale=<0.1 0.1 0.1>
    componentLinearVelScale=<0.1 0.1 0.1>
    physicsBodyDefinitions [{
      boundBone {boneName=dyn_bootstrap_01_l}
      boxExtents=<5 5 5>
      constraintSetup {angularLimitsMax=<5 0 30>, angularLimitsMin=<-5 0 -30>, angularTarget=<0 0 0>}
      localJointOffset=<0 -1 0>
     }]
    relativeSpaceBone {boneName=spine_05}
   }
   n7 = Anim Dynamics - Bone: dyn_bootstrap_01_r!(ComponentPose=@n6, Alpha=1, GravityOverride:Vector="0,0,-300", ExternalForce="0,0,0") -> (Pose) {
    angularDampingOverride=1
    angularSpringConstant=1000
    bAngularSpring=true
    bOverrideAngularDamping=true
    bUseGravityOverride=true
    alphaBoolBlend {blendOption=Linear}
    boundBone {boneName=dyn_bootstrap_01_r}
    componentAppliedLinearAccClamp=<2000 2000 2000>
    componentLinearAccScale=<0.1 0.1 0.1>
    componentLinearVelScale=<0.1 0.1 0.1>
    physicsBodyDefinitions [{
      boundBone {boneName=dyn_bootstrap_01_r}
      boxExtents=<5 5 5>
      constraintSetup {angularLimitsMax=<5 0 30>, angularLimitsMin=<-5 0 -30>, angularTarget=<0 0 0>}
      localJointOffset=<0 -1 0>
     }]
    relativeSpaceBone {boneName=spine_05}
   }
   Output Pose(Result=Component To Local(ComponentPose=@n7)) {layerGroup=None}
  ]}
 EventGraph {nodes [
   n1 = Event BlueprintUpdateAnimation()
   > Set Delta Time X(Delta Time X=@n1.DeltaTimeX)
   > Branch()
   n2 = Get Skel Mesh()
   > Set LeftArmBendEulerYAlpha(LeftArmBendEulerYAlpha:double=GetAlphaFromReferenceTransformDelta(SkeletalMeshComponent=@n2, BoneName=upperarm_l, BaseName=clavicle_l, DeltaComponentType=EulerY, InputRange=(X=-35,Y=20), TwistAxis="1,0,0")) -> (Output_Get:double)
   > Set RightArmBendEulerYAlpha(RightArmBendEulerYAlpha:double=GetAlphaFromReferenceTransformDelta(SkeletalMeshComponent=@n2, BoneName=upperarm_r, BaseName=clavicle_r, DeltaComponentType=EulerY, InputRange=(X=-35,Y=20), TwistAxis="1,0,0")) -> (Output_Get:double)
   Event BlueprintInitializeAnimation()
   > Set Skel Mesh(Skel Mesh=GetOwnerSkeletalMeshComponent())
  ]}
]
variables [
 Delta Time X {type=number}
 LeftArmBendEulerYAlpha {type=number}
 PlayerPawn {type=FortPlayerPawnAthena}
 RightArmBendEulerYAlpha {type=number}
 Skel Mesh {type=SkeletalMeshComponent}
]
