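# Post-copy AnimBP for the MusketSlinger Style face accessory part, adding spring-driven hair motion.
# It copies the attached parent's pose, layering facial jaw and chin bones from the Head mesh cached on customization load.
# Chained AnimDynamics nodes simulate front and back hair bones at half alpha, some with head-sphere limits.
# In the front end, the Frontend_Idle_Bypass curve weights DefaultSlot animation over two front-right hair bones.
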AnimBlueprint /BRCosmetics/Characters/Player/Male/Medium/Bodies/M_MED_MusketSlinger_Style/Meshes/Parts/M_MED_MusketSlinger_Style_FaceAcc_AnimBP
target_skeleton=&/M_MED_MusketSlinger_Style_FaceAcc_Skeleton
graphs [
 AnimGraph {nodes [
   Output Pose(Result=Layered blend per bone(
     BasePose=Use cached pose 'StartPose'()
     BlendPoses_0=Slot 'DefaultSlot'()
     BlendWeights_0=Get Custom Idle is Playing() -> (Custom Idle is Playing:double)
    ) {layerSetup [{branchFilters [{boneName=dyn_hair_fr_r_1}, {boneName=dyn_hair_fr_r_2}]}]}) {layerGroup=None}
   n1 = Layered blend per bone(
    BasePose=Copy Pose From Mesh() {bUseAttachedParent=true}
    BlendPoses_0=Copy Pose From Mesh(SourceMeshComponent=Get Head Mesh() -> (Head Mesh:Skeletal Mesh Component Object Reference))
   ) {layerSetup [{branchFilters [
       {boneName=FACIAL_C_Jaw}
       {boneName=FACIAL_C_Chin}
       {boneName=FACIAL_C_Jawline}
       {boneName=FACIAL_L_ChinSide}
       {boneName=FACIAL_R_ChinSide}
       {boneName=FACIAL_C_UnderChin}
       {boneName=FACIAL_L_Jawline}
       {boneName=FACIAL_L_Masseter}
       {boneName=FACIAL_L_UnderChin}
       {boneName=FACIAL_R_Jawline}
       {boneName=FACIAL_R_Masseter}
       {boneName=FACIAL_R_UnderChin}
      ]}]}
   Save cached pose 'Start_Cache'(Pose=@n1) {cacheName=Start_Cache}
   n2 = Anim Dynamics - Bone: dyn_hair_fr_1!(ComponentPose=Local To Component(LocalPose=Use cached pose 'Start_Cache'()), ExternalForce="0,0,0") -> (Pose) {
    alpha=0.5
    angularDampingOverride=1
    angularSpringConstant=1000
    bAngularSpring=true
    bOverrideAngularDamping=true
    bUseGravityOverride=true
    bUsePlanarLimit=false
    bUseSphericalLimits=true
    gravityScale=0.35
    alphaBoolBlend {blendOption=Linear}
    boundBone {boneName=dyn_hair_fr_1}
    componentAppliedLinearAccClamp=<2000 2000 2000>
    componentLinearAccScale=<0.2 0.2 0.2>
    componentLinearVelScale=<0.3 0.3 0.3>
    gravityOverride {z=-200}
    physicsBodyDefinitions [{
      boundBone {boneName=dyn_hair_fr_1}
      boxExtents=<4 3 3>
      constraintSetup {angularLimitsMax=<0 30 15>, angularLimitsMin=<0 0 -35>, angularTarget=<0 0 0>}
      localJointOffset=<1 0 0>
     }]
    relativeSpaceBone {boneName=spine_05}
    sphericalLimits [{limitRadius=8.2, drivingBone {boneName=head}, sphereLocalOffset=<7.313 2.2115 0>}]
   }
   n3 = Anim Dynamics - Bone: dyn_hair_fr_2!(ComponentPose=@n2, ExternalForce="0,0,0") -> (Pose) {
    alpha=0.5
    angularDampingOverride=1
    angularSpringConstant=1000
    bAngularSpring=true
    bOverrideAngularDamping=true
    bUseGravityOverride=true
    bUsePlanarLimit=false
    bUseSphericalLimits=true
    gravityScale=0.35
    alphaBoolBlend {blendOption=Linear}
    boundBone {boneName=dyn_hair_fr_2}
    componentAppliedLinearAccClamp=<2000 2000 2000>
    componentLinearAccScale=<0.3 0.3 0.3>
    componentLinearVelScale=<0.5 0.5 0.5>
    gravityOverride {z=-300}
    physicsBodyDefinitions [{
      boundBone {boneName=dyn_hair_fr_2}
      boxExtents=<2 4 4>
      constraintSetup {angularLimitsMax=<10 30 45>, angularLimitsMin=<-10 -30 -45>, angularTarget=<0 0 0>}
      localJointOffset=<0.7 0 0>
     }]
    relativeSpaceBone {boneName=spine_05}
    sphericalLimits [{limitRadius=8.2, drivingBone {boneName=head}, sphereLocalOffset=<7.0971 2.7981 0>}]
   }
   n4 = Anim Dynamics - Bone: dyn_hair_fr_3!(ComponentPose=@n3, ExternalForce="0,0,0") -> (Pose) {
    alpha=0.5
    angularDampingOverride=1
    angularSpringConstant=1000
    bAngularSpring=true
    bOverrideAngularDamping=true
    bUseGravityOverride=true
    bUsePlanarLimit=false
    bUseSphericalLimits=true
    gravityScale=0.35
    alphaBoolBlend {blendOption=Linear}
    boundBone {boneName=dyn_hair_fr_3}
    componentAppliedLinearAccClamp=<2000 2000 2000>
    componentLinearAccScale=<0.2 0.2 0.2>
    componentLinearVelScale=<0.3 0.3 0.3>
    gravityOverride {z=-180}
    physicsBodyDefinitions [{
      boundBone {boneName=dyn_hair_fr_3}
      boxExtents=<4 4 4>
      constraintSetup {angularLimitsMax=<10 30 30>, angularLimitsMin=<-10 -30 -30>, angularTarget=<0 0 0>}
      localJointOffset=<1 0.5 0>
     }]
    relativeSpaceBone {boneName=spine_05}
    sphericalLimits [{limitRadius=8, drivingBone {boneName=head}, sphereLocalOffset=<6.9747 3.6578 0.239>}]
   }
   n5 = Anim Dynamics - Bone: dyn_hair_fr_r_1!(ComponentPose=@n4, ExternalForce="0,0,0") -> (Pose) {
    alpha=0.5
    angularDampingOverride=0.9999
    angularSpringConstant=700
    bAngularSpring=true
    bOverrideAngularDamping=true
    bUseGravityOverride=true
    bUsePlanarLimit=false
    bUseSphericalLimits=true
    gravityScale=0.35
    alphaBoolBlend {blendOption=Linear}
    boundBone {boneName=dyn_hair_fr_r_1}
    componentAppliedLinearAccClamp=<2000 2000 2000>
    componentLinearAccScale=<0.2 0.2 0.2>
    componentLinearVelScale=<0.3 0.3 0.3>
    gravityOverride {z=-150}
    physicsBodyDefinitions [{
      boundBone {boneName=dyn_hair_fr_r_1}
      boxExtents=<4 4 4>
      constraintSetup {angularLimitsMax=<10 30 30>, angularLimitsMin=<-10 -30 -30>, angularTarget=<0 0 0>}
      localJointOffset=<1 0.5 0>
     }]
    relativeSpaceBone {boneName=spine_05}
    sphericalLimits [{limitRadius=8, drivingBone {boneName=head}, sphereLocalOffset=<6.7179 2.6666 0.8247>}]
   }
   n6 = Anim Dynamics - Bone: dyn_hair_fr_r_2!(ComponentPose=@n5, GravityOverride:Vector="0,0,-300", ExternalForce="0,0,0") -> (Pose) {
    alpha=0.5
    angularDampingOverride=1
    angularSpringConstant=1000
    bAngularSpring=true
    bOverrideAngularDamping=true
    bUseGravityOverride=true
    bUsePlanarLimit=false
    bUseSphericalLimits=true
    alphaBoolBlend {blendOption=Linear}
    boundBone {boneName=dyn_hair_fr_r_2}
    componentAppliedLinearAccClamp=<2000 2000 2000>
    componentLinearAccScale=<0.4 0.4 0.4>
    componentLinearVelScale=<0.4 0.4 0.4>
    physicsBodyDefinitions [{
      boundBone {boneName=dyn_hair_fr_r_2}
      boxExtents=<5 5 5>
      constraintSetup {angularLimitsMax=<10 30 30>, angularLimitsMin=<-10 -30 -30>, angularTarget=<0 0 0>}
      localJointOffset=<1 0.5 0>
     }]
    relativeSpaceBone {boneName=spine_05}
    sphericalLimits [{limitRadius=8, drivingBone {boneName=head}, sphereLocalOffset=<7.6861 2.6013 1.3825>}]
   }
   n7 = Anim Dynamics - Bone: dyn_hair_fr_up_1!(ComponentPose=@n6, GravityOverride:Vector="0,0,-300", ExternalForce="0,0,0") -> (Pose) {
    alpha=0.5
    angularDampingOverride=1
    angularSpringConstant=1000
    bAngularSpring=true
    bOverrideAngularDamping=true
    bUseGravityOverride=true
    bUsePlanarLimit=false
    bUseSphericalLimits=true
    alphaBoolBlend {blendOption=Linear}
    boundBone {boneName=dyn_hair_fr_up_1}
    componentAppliedLinearAccClamp=<2000 2000 2000>
    componentLinearAccScale=<0.4 0.4 0.4>
    componentLinearVelScale=<0.4 0.4 0.4>
    physicsBodyDefinitions [{
      boundBone {boneName=dyn_hair_fr_up_1}
      boxExtents=<7 7 7>
      constraintSetup {angularLimitsMax=<10 30 30>, angularLimitsMin=<-10 -30 -30>, angularTarget=<0 0 0>}
      localJointOffset=<1 0.5 0>
     }]
    relativeSpaceBone {boneName=spine_05}
   }
   n8 = Anim Dynamics - Bone: dyn_hair_back_r_1!(ComponentPose=@n7, GravityOverride:Vector="0,0,-300", ExternalForce="0,0,0") -> (Pose) {
    alpha=0.5
    angularDampingOverride=1
    angularSpringConstant=1000
    bAngularSpring=true
    bOverrideAngularDamping=true
    bUseGravityOverride=true
    alphaBoolBlend {blendOption=Linear}
    boundBone {boneName=dyn_hair_back_r_1}
    componentAppliedLinearAccClamp=<2000 2000 2000>
    componentLinearAccScale=<0.4 0.4 0.4>
    componentLinearVelScale=<0.4 0.4 0.4>
    physicsBodyDefinitions [{
      boundBone {boneName=dyn_hair_back_r_1}
      boxExtents=<7 7 7>
      constraintSetup {angularLimitsMax=<10 30 30>, angularLimitsMin=<-10 -30 -30>, angularTarget=<0 0 0>}
      localJointOffset=<1 0.5 0>
     }]
    relativeSpaceBone {boneName=spine_05}
   }
   n9 = Anim Dynamics - Bone: dyn_hair_back_1!(ComponentPose=@n8, GravityOverride:Vector="0,0,0", ExternalForce="0,0,0") -> (Pose) {
    alpha=0.5
    angularDampingOverride=1
    angularSpringConstant=1000
    bAngularSpring=true
    bOverrideAngularDamping=true
    bUseGravityOverride=true
    alphaBoolBlend {blendOption=Linear}
    boundBone {boneName=dyn_hair_back_1}
    componentAppliedLinearAccClamp=<2000 2000 2000>
    componentLinearAccScale=<0.4 0.4 0.4>
    componentLinearVelScale=<0.4 0.4 0.4>
    physicsBodyDefinitions [{
      boundBone {boneName=dyn_hair_back_1}
      boxExtents=<3 3 3>
      constraintSetup {angularLimitsMax=<10 30 30>, angularLimitsMin=<-10 -30 -30>, angularTarget=<0 0 0>}
      localJointOffset=<1 0.5 0>
     }]
    relativeSpaceBone {boneName=spine_05}
   }
   n10 = Anim Dynamics - Bone: dyn_hair_back_2!(ComponentPose=@n9, GravityOverride:Vector="0,0,0", ExternalForce="0,0,0") -> (Pose) {
    alpha=0.5
    angularDampingOverride=1
    angularSpringConstant=1000
    bAngularSpring=true
    bOverrideAngularDamping=true
    bUseGravityOverride=true
    alphaBoolBlend {blendOption=Linear}
    boundBone {boneName=dyn_hair_back_2}
    componentAppliedLinearAccClamp=<2000 2000 2000>
    componentLinearAccScale=<0.4 0.4 0.4>
    componentLinearVelScale=<0.4 0.4 0.4>
    physicsBodyDefinitions [{
      boundBone {boneName=dyn_hair_back_2}
      boxExtents=<5 5 5>
      constraintSetup {angularLimitsMax=<10 30 30>, angularLimitsMin=<-10 -30 -30>, angularTarget=<0 0 0>}
      localJointOffset=<1 0.5 0>
     }]
    relativeSpaceBone {boneName=spine_05}
   }
   n11 = Anim Dynamics - Bone: dyn_hair_back_3!(ComponentPose=@n10, GravityOverride:Vector="0,0,0", ExternalForce="0,0,0") -> (Pose) {
    alpha=0.5
    angularDampingOverride=1
    angularSpringConstant=1000
    bAngularSpring=true
    bOverrideAngularDamping=true
    bUseGravityOverride=true
    alphaBoolBlend {blendOption=Linear}
    boundBone {boneName=dyn_hair_back_3}
    componentAppliedLinearAccClamp=<2000 2000 2000>
    componentLinearAccScale=<0.4 0.4 0.4>
    componentLinearVelScale=<0.4 0.4 0.4>
    physicsBodyDefinitions [{
      boundBone {boneName=dyn_hair_back_3}
      boxExtents=<8 8 8>
      constraintSetup {angularLimitsMax=<10 30 30>, angularLimitsMin=<-10 -30 -30>, angularTarget=<0 0 0>}
      localJointOffset=<1 0.5 0>
     }]
    relativeSpaceBone {boneName=spine_05}
   }
   n12 = Anim Dynamics - Bone: dyn_hair_back_up_1!(ComponentPose=@n11, GravityOverride:Vector="0,0,0", ExternalForce="0,0,0") -> (Pose) {
    alpha=0.5
    angularDampingOverride=1
    angularSpringConstant=1000
    bAngularSpring=true
    bOverrideAngularDamping=true
    bUseGravityOverride=true
    alphaBoolBlend {blendOption=Linear}
    boundBone {boneName=dyn_hair_back_up_1}
    componentAppliedLinearAccClamp=<2000 2000 2000>
    componentLinearAccScale=<0.4 0.4 0.4>
    componentLinearVelScale=<0.4 0.4 0.4>
    physicsBodyDefinitions [{
      boundBone {boneName=dyn_hair_back_up_1}
      boxExtents=<7 7 7>
      constraintSetup {angularLimitsMax=<10 30 30>, angularLimitsMin=<-10 -30 -30>, angularTarget=<0 0 0>}
      localJointOffset=<1 0.5 0>
     }]
    relativeSpaceBone {boneName=spine_05}
   }
   n13 = Anim Dynamics - Bone: dyn_hair_back_4!(ComponentPose=@n12, GravityOverride:Vector="0,0,0", ExternalForce="0,0,0") -> (Pose) {
    alpha=0.5
    angularDampingOverride=1
    angularSpringConstant=1000
    bAngularSpring=true
    bOverrideAngularDamping=true
    bUseGravityOverride=true
    alphaBoolBlend {blendOption=Linear}
    boundBone {boneName=dyn_hair_back_4}
    componentAppliedLinearAccClamp=<2000 2000 2000>
    componentLinearAccScale=<0.4 0.4 0.4>
    componentLinearVelScale=<0.4 0.4 0.4>
    physicsBodyDefinitions [{
      boundBone {boneName=dyn_hair_back_4}
      boxExtents=<3 3 3>
      constraintSetup {angularLimitsMax=<10 30 30>, angularLimitsMin=<-10 -30 -30>, angularTarget=<0 0 0>}
      localJointOffset=<1 0.5 0>
     }]
    relativeSpaceBone {boneName=spine_05}
   }
   Save cached pose 'StartPose'(Pose=Component To Local(ComponentPose=@n13)) {cacheName=StartPose}
  ]}
 EventGraph {nodes [
   Event BlueprintUpdateAnimation()
   > Branch(Condition=Get bIsInFrontEnd())
   > Set Custom Idle is Playing(Custom Idle is Playing:double=GetCurveValue(self=GetOwnerFrontEndAnimInstance(), CurveName=Frontend_Idle_Bypass)) -> (Output_Get:double)
   n1 = Event OnCharacterCustomizationLoaded()
   > Set Head Mesh(Head Mesh:Skeletal Mesh Component Object Reference=Get Array Item with EFort Custom Part Type(Array=@n1.CustomMeshParts, PartType=Head)) -> (Output_Get:Skeletal Mesh Component Object Reference)
  ]}
]
variables [
 Custom Idle is Playing {type=number}
 Head Mesh {type=SkeletalMeshComponent}
]
